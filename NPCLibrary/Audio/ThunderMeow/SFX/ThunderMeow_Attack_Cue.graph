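# Sound cue for the ThunderMeow NPC's attack sound, routed to the weapon melee sound class.
# A single Thundermeow_Spin_01 wave player feeds a Modulator that adds slight random upward pitch variation.
# It references the ThunderMeow melee attack attenuation asset and overrides concurrency to cap simultaneous instances.
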SoundCue /NPCLibrary/Audio/ThunderMeow/SFX/ThunderMeow_Attack_Cue
root=Modulator_0
nodes [
 {id=Modulator_0, children [WavePlayer_10], properties {pitchMax=1.02, pitchMin=1}}
 {id=WavePlayer_10, properties {soundWaveAssetPtr=&/Thundermeow_Spin_01}}
]
properties {
 attenuationSettings=/NPCLibrary/Audio/ThunderMeow/SFX/Settings/ThunderMeow_MeleeAttack_Att
 bOverrideConcurrency=true
 maxDistance=3800
 soundClassObject=&Fort_WeaponMelee
 volumeMultiplier=0.76
 attenuationOverrides {spatializationAlgorithm=SPATIALIZATION_HRTF}
 concurrencyOverrides {maxCount=6}
 concurrencySet [null]
}
