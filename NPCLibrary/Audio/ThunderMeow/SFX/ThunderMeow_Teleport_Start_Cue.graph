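# ThunderMeow teleport start sound cue, playing the Thundermeow_Start_01 wave for the start of the NPC's teleport ability.
# A modulator at the root randomizes pitch slightly around normal on each play of the single wave player.
# It uses the ThunderMeow ability attenuation and concurrency settings, is routed to a melee weapon sound class, and plays with a slight volume boost.
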SoundCue /NPCLibrary/Audio/ThunderMeow/SFX/ThunderMeow_Teleport_Start_Cue
root=Modulator_0
nodes [
 {id=Modulator_0, children [WavePlayer_1], properties {pitchMax=1.1, pitchMin=0.99, volumeMin=1}}
 {id=WavePlayer_1, properties {soundWaveAssetPtr=&/Thundermeow_Start_01}}
]
properties {
 attenuationSettings=/NPCLibrary/Audio/ThunderMeow/SFX/Settings/ThunderMeow_Ability_Att
 maxDistance=4000
 priority=0.5
 soundClassObject=&Fort_WeaponMelee_PSM
 volumeMultiplier=1.2
 concurrencySet [&/ThunderMeow_Concurrency]
}
